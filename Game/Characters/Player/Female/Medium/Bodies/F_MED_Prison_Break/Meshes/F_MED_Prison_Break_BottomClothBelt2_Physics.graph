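# Physics asset for the female medium Prison Break bottom-body outfit, covering its dangling belt, back cloth and two front loincloth flaps.
# Swing-limited, zero-twist capsule chains hang the back cloth and flaps from the pelvis, with the belt branching off the third back-cloth bone.
# Pelvis and leg capsules with free angular constraints provide body collision shapes.
# A huge root capsule offset downward appears to act as a ground proxy.
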PhysicsAsset /Game/Characters/Player/Female/Medium/Bodies/F_MED_Prison_Break/Meshes/F_MED_Prison_Break_BottomClothBelt2_Physics
bodies {
 calf_l [{shapeType=Capsule, length=30.0904, radius=10.3434, shapeName=calf_l_capsule, center=<-21.9407 0 0>, rotation=<rot 90 0 0>}]
 calf_r [{shapeType=Capsule, length=36.4466, radius=9.2336, shapeName=calf_l_capsule, center=<21.9407 0 0>, rotation=<rot 90 80.7735 -103.4968>}]
 dyn_belt2_01 [{shapeType=Capsule, length=9.6594, radius=1.7518, shapeName=dyn_belt2_01_capsule, center=<4.2046 1 0>, rotation=<rot 90 0 0>}]
 dyn_belt2_02 [{shapeType=Capsule, length=13.2638, radius=2.1158, shapeName=dyn_belt2_02_capsule, center=<8.4762 1.5 0>, rotation=<rot 90 0 0>}]
 dyn_belt2_03 [{shapeType=Capsule, length=9.7208, radius=3, shapeName=dyn_belt2_03_capsule, center=<6.9186 0 0>, rotation=<rot 90 0 0>}]
 dyn_belt2_04 [{shapeType=Capsule, length=3, radius=3, shapeName=dyn_belt2_04_capsule, center=<3.3396 0 0>, rotation=<rot 90 0 0>}]
 dyn_belt2_05 [{shapeType=Capsule, length=0, radius=3, shapeName=dyn_belt2_05_capsule, center=<1.2306 0 0>, rotation=<rot 90 0 0>}]
 dyn_belt2_06 [{shapeType=Capsule, length=0.2595, radius=3, shapeName=dyn_belt2_06_capsule, center=<3.3396 0 0>, rotation=<rot 90 0 0>}]
 dyn_belt2_MSTR [{shapeType=Capsule, length=0.5, radius=0.5, shapeName=dyn_belt2_MSTR_capsule}]
 dyn_clothBack_01 [{shapeType=Capsule, length=13.5847, radius=3, shapeName=dyn_clothBack_01_capsule, center=<4.6383 -1 0>}]
 dyn_clothBack_02 [{shapeType=Capsule, length=12.9747, radius=2.7881, shapeName=dyn_clothBack_02_capsule, center=<3.9834 -1 0>}]
 dyn_clothBack_03 [{shapeType=Capsule, length=12.5224, radius=2.5, shapeName=dyn_clothBack_03_capsule, center=<6.0583 -1.7 0>}]
 dyn_clothBack_04 [{shapeType=Capsule, length=8.9166, radius=2.75, shapeName=dyn_clothBack_04_capsule, center=<4.5875 -2.5 0>}]
 dyn_clothBack_MSTR [{shapeType=Capsule, length=0.5, radius=0.5, shapeName=dyn_clothBack_MSTR_capsule}]
 dyn_loinClothFront_l_01 [{shapeType=Capsule, length=0.5, radius=0.5, shapeName=dyn_loinClothFront_l_01_capsule}]
 dyn_loinClothFront_l_02 [{
   shapeType=Capsule
   length=5.5
   radius=2.75
   shapeName=dyn_loinClothFront_l_02_capsule
   center=<3.8103 2.3 -1>
   rotation=<rot 80.4916 0.0001 0>
  }]
 dyn_loinClothFront_r_01 [{shapeType=Capsule, length=0.5, radius=0.5, shapeName=dyn_loinClothFront_r_01_capsule}]
 dyn_loinClothFront_r_02 [{
   shapeType=Capsule
   length=5.25
   radius=3
   shapeName=dyn_loinClothFront_r_02_capsule
   center=<3.6569 2.3 1.4>
   rotation=<rot 78.8314 -179.9999 -180>
  }]
 foot_l [{
   shapeType=Capsule
   length=14.9828
   radius=6.5087
   shapeName=foot_l_capsule
   center=<-3.9243 6.4333 0.0498>
   rotation=<rot 3.0103 4.8392 -84.3336>
  }]
 foot_r [{
   shapeType=Capsule
   length=14.9828
   radius=6.5087
   shapeName=foot_l_capsule
   center=<3.9243 -6.4333 -0.0498>
   rotation=<rot 3.0103 4.8392 95.6664>
  }]
 pelvis [{shapeType=Capsule, length=4.8512, radius=11.4881, shapeName=pelvis_capsule, center=<-2.9558 0 0>}]
 root [{shapeType=Capsule, length=0.505, radius=300, shapeName=root_capsule, center=<0 -43.1161 -297.2333>, rotation=<rot 0 -90 0>}]
 thigh_l [{shapeType=Capsule, length=30.5095, radius=9.5063, shapeName=thigh_l_capsule, center=<-16.4516 0 0>, rotation=<rot 90 0 0>}]
 thigh_r [{shapeType=Capsule, length=37.1228, radius=8.1393, shapeName=thigh_l_capsule, center=<16.4516 1.4222 0>, rotation=<rot 90 0 180>}]
}
constraints [
 {bone1Name=dyn_belt2_06, bone2Name=dyn_belt2_05, swing1LimitDegrees=45, swing2LimitDegrees=0, twistLimitDegrees=0}
 {bone1Name=dyn_belt2_05, bone2Name=dyn_belt2_04, swing1LimitDegrees=45, swing2LimitDegrees=0, twistLimitDegrees=0}
 {bone1Name=dyn_belt2_04, bone2Name=dyn_belt2_03, swing1LimitDegrees=50, swing2LimitDegrees=50, twistLimitDegrees=0}
 {bone1Name=dyn_belt2_03, bone2Name=dyn_belt2_02, swing1LimitDegrees=55, swing2LimitDegrees=40, twistLimitDegrees=0}
 {bone1Name=dyn_belt2_02, bone2Name=dyn_belt2_01, swing1LimitDegrees=60, swing2LimitDegrees=30, twistLimitDegrees=0}
 {bone1Name=dyn_belt2_01, bone2Name=dyn_belt2_MSTR, swing1LimitDegrees=10, swing2LimitDegrees=20, twistLimitDegrees=0}
 {bone1Name=dyn_belt2_MSTR, bone2Name=dyn_clothBack_03, swing1LimitDegrees=5, swing2LimitDegrees=5, twistLimitDegrees=0}
 {bone1Name=dyn_clothBack_03, bone2Name=dyn_clothBack_02, swing1LimitDegrees=30, swing2LimitDegrees=15, twistLimitDegrees=0}
 {bone1Name=dyn_clothBack_02, bone2Name=dyn_clothBack_01, swing1LimitDegrees=20, swing2LimitDegrees=15, twistLimitDegrees=0}
 {bone1Name=dyn_clothBack_01, bone2Name=dyn_clothBack_MSTR, swing1LimitDegrees=15, swing2LimitDegrees=15, twistLimitDegrees=0}
 {
  bone1Name=dyn_clothBack_MSTR
  bone2Name=pelvis
  swing1LimitDegrees=45
  swing1Motion=Free
  swing2LimitDegrees=45
  swing2Motion=Free
  twistLimitDegrees=45
  twistMotion=Free
 }
 {bone1Name=dyn_clothBack_04, bone2Name=dyn_clothBack_03, swing1LimitDegrees=30, swing2LimitDegrees=30, twistLimitDegrees=0}
 {
  bone1Name=thigh_l
  bone2Name=pelvis
  swing1LimitDegrees=45
  swing1Motion=Free
  swing2LimitDegrees=45
  swing2Motion=Free
  twistLimitDegrees=45
  twistMotion=Free
 }
 {
  bone1Name=calf_l
  bone2Name=thigh_l
  swing1LimitDegrees=45
  swing1Motion=Free
  swing2LimitDegrees=45
  swing2Motion=Free
  twistLimitDegrees=45
  twistMotion=Free
 }
 {
  bone1Name=calf_r
  bone2Name=thigh_r
  swing1LimitDegrees=45
  swing1Motion=Free
  swing2LimitDegrees=45
  swing2Motion=Free
  twistLimitDegrees=45
  twistMotion=Free
 }
 {
  bone1Name=thigh_r
  bone2Name=pelvis
  swing1LimitDegrees=45
  swing1Motion=Free
  swing2LimitDegrees=45
  swing2Motion=Free
  twistLimitDegrees=45
  twistMotion=Free
 }
 {
  bone1Name=pelvis
  bone2Name=root
  swing1LimitDegrees=45
  swing1Motion=Free
  swing2LimitDegrees=45
  swing2Motion=Free
  twistLimitDegrees=45
  twistMotion=Free
 }
 {
  bone1Name=foot_l
  bone2Name=calf_l
  swing1LimitDegrees=45
  swing1Motion=Free
  swing2LimitDegrees=45
  swing2Motion=Free
  twistLimitDegrees=45
  twistMotion=Free
 }
 {
  bone1Name=foot_r
  bone2Name=calf_r
  swing1LimitDegrees=45
  swing1Motion=Free
  swing2LimitDegrees=45
  swing2Motion=Free
  twistLimitDegrees=45
  twistMotion=Free
 }
 {bone1Name=dyn_loinClothFront_l_01, bone2Name=pelvis, swing1LimitDegrees=90, swing2LimitDegrees=30, twistLimitDegrees=0}
 {bone1Name=dyn_loinClothFront_l_02, bone2Name=dyn_loinClothFront_l_01, swing1LimitDegrees=60, swing2LimitDegrees=10, twistLimitDegrees=0}
 {bone1Name=dyn_loinClothFront_r_01, bone2Name=pelvis, swing1LimitDegrees=90, swing2LimitDegrees=30, twistLimitDegrees=0}
 {bone1Name=dyn_loinClothFront_r_02, bone2Name=dyn_loinClothFront_r_01, swing1LimitDegrees=60, swing2LimitDegrees=10, twistLimitDegrees=0}
]
