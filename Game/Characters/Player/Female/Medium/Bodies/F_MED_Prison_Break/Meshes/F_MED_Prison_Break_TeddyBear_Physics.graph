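# Physics asset for the dangling teddy bear on the female medium Prison Break player body, using capsule bodies only.
# An anchor chain leads into the bear's torso, which branches to head, arms and two-segment legs with limited swing and zero twist below the anchor.
# Left thigh and calf capsules and two back-cloth capsules form separate pairs, with the cloth joint fully free in rotation.
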PhysicsAsset /Game/Characters/Player/Female/Medium/Bodies/F_MED_Prison_Break/Meshes/F_MED_Prison_Break_TeddyBear_Physics
bodies {
 calf_l [{shapeType=Capsule, length=38.2089, radius=5.5176, center=<-20.719 -2.9271 0>, rotation=<rot 90 0 0>}]
 dyn_clothBack_03 [{
   shapeType=Capsule
   length=7.6284
   radius=5.0609
   shapeName=dyn_clothBack_03_capsule
   center=<4.2452 -2.3885 -0.113>
   rotation=<rot -0.4952 -90 1.4935>
  }]
 dyn_clothBack_04 [{shapeType=Capsule, length=6.6484, radius=4.6464, shapeName=dyn_clothBack_04_capsule, center=<2.6607 -3.1735 0>, rotation=<rot 0 -90 0>}]
 dyn_teddyAnchor_01 [{shapeType=Capsule, length=6.0949, radius=2, shapeName=dyn_teddyAnchor_01_capsule, center=<5.5601 0 0>, rotation=<rot 90 0 0>}]
 dyn_teddyAnchor_MSTR [{shapeType=Capsule, length=0.5, radius=0.5, shapeName=dyn_teddyAnchor_MSTR_capsule}]
 dyn_teddyArm_l_01 [{shapeType=Capsule, length=1.616, radius=2.6271, shapeName=dyn_teddyArm_l_01_capsule, center=<6.8964 0 0>, rotation=<rot 90 0 0>}]
 dyn_teddyArm_r_01 [{shapeType=Capsule, length=2.4329, radius=2.681, shapeName=dyn_teddyArm_r_01_capsule, center=<4.0918 0 0>, rotation=<rot 90 0 0>}]
 dyn_teddyBody_01 [{shapeType=Capsule, length=3.2311, radius=3.9804, shapeName=dyn_teddyBody_01_capsule, center=<1.7089 0 0>, rotation=<rot 90 0 0>}]
 dyn_teddyBody_02 [{shapeType=Capsule, length=1.6735, radius=4.4094, shapeName=dyn_teddyBody_02_capsule, center=<1.2665 0 0>}]
 dyn_teddyBody_03 [{shapeType=Capsule, length=1.6735, radius=4.4094, shapeName=dyn_teddyBody_03_capsule, center=<1.2665 0 0>}]
 dyn_teddyBody_MSTR [{shapeType=Capsule, length=2, radius=2, shapeName=dyn_teddyBody_MSTR_capsule, rotation=<rot 90 0 0>}]
 dyn_teddyHead_01 [{shapeType=Capsule, length=0, radius=5.8154, shapeName=dyn_teddyHead_01_capsule, center=<5.1662 0 0>, rotation=<rot 90 0 0>}]
 dyn_teddyLeg_l_01 [{shapeType=Capsule, length=3.3234, radius=2.4616, shapeName=dyn_teddyLeg_l_01_capsule, center=<2.643 0 0>, rotation=<rot 90 0 0>}]
 dyn_teddyLeg_l_02 [{
   shapeType=Capsule
   length=5.3459
   radius=2.9611
   shapeName=dyn_teddyLeg_l_02_capsule
   center=<5.4255 0.99 0>
   rotation=<rot 90 -172.875 176.1427>
  }]
 dyn_teddyLeg_r_01 [{shapeType=Capsule, length=3.7614, radius=2.0235, shapeName=dyn_teddyLeg_r_01_capsule, center=<2.643 0 0>, rotation=<rot 90 0 0>}]
 dyn_teddyLeg_r_02 [{shapeType=Capsule, length=3.3234, radius=2.4616, shapeName=dyn_teddyLeg_r_02_capsule, center=<3.179 0 0>, rotation=<rot 90 0 0>}]
 thigh_l [{shapeType=Capsule, length=28.2043, radius=7.0698, center=<-20.7225 0.4027 1.2735>, rotation=<rot 90 0 0>}]
}
constraints [
 {bone1Name=calf_l, bone2Name=thigh_l, swing1LimitDegrees=45, swing2LimitDegrees=45, twistLimitDegrees=45}
 {bone1Name=dyn_teddyAnchor_01, bone2Name=dyn_teddyAnchor_MSTR, swing1LimitDegrees=30, swing2LimitDegrees=30, twistLimitDegrees=20}
 {bone1Name=dyn_teddyBody_MSTR, bone2Name=dyn_teddyAnchor_01, swing1LimitDegrees=5, swing2LimitDegrees=5, twistLimitDegrees=0}
 {bone1Name=dyn_teddyBody_01, bone2Name=dyn_teddyBody_MSTR, swing1LimitDegrees=5, swing2LimitDegrees=5, twistLimitDegrees=0}
 {bone1Name=dyn_teddyArm_l_01, bone2Name=dyn_teddyBody_01, swing1LimitDegrees=40, swing2LimitDegrees=40, twistLimitDegrees=0}
 {bone1Name=dyn_teddyArm_r_01, bone2Name=dyn_teddyBody_01, swing1LimitDegrees=20, swing2LimitDegrees=20, twistLimitDegrees=0}
 {bone1Name=dyn_teddyBody_02, bone2Name=dyn_teddyBody_01, swing1LimitDegrees=20, swing2LimitDegrees=20, twistLimitDegrees=0}
 {bone1Name=dyn_teddyBody_03, bone2Name=dyn_teddyBody_02, swing1LimitDegrees=20, swing2LimitDegrees=20, twistLimitDegrees=0}
 {bone1Name=dyn_teddyLeg_l_01, bone2Name=dyn_teddyBody_03, swing1LimitDegrees=40, swing2LimitDegrees=40, twistLimitDegrees=0}
 {bone1Name=dyn_teddyLeg_l_02, bone2Name=dyn_teddyLeg_l_01, swing1LimitDegrees=40, swing2LimitDegrees=40, twistLimitDegrees=0}
 {bone1Name=dyn_teddyLeg_r_01, bone2Name=dyn_teddyBody_03, swing1LimitDegrees=40, swing2LimitDegrees=40, twistLimitDegrees=0}
 {bone1Name=dyn_teddyLeg_r_02, bone2Name=dyn_teddyLeg_r_01, swing1LimitDegrees=40, swing2LimitDegrees=40, twistLimitDegrees=0}
 {bone1Name=dyn_teddyHead_01, bone2Name=dyn_teddyBody_01, swing1LimitDegrees=20, swing2LimitDegrees=20, twistLimitDegrees=0}
 {
  bone1Name=dyn_clothBack_04
  bone2Name=dyn_clothBack_03
  swing1LimitDegrees=45
  swing1Motion=Free
  swing2LimitDegrees=45
  swing2Motion=Free
  twistLimitDegrees=45
  twistMotion=Free
 }
]
