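# Physics asset for the front top buckles on the medium female Prison Break player body, defining capsule bodies and joints for secondary motion.
# Left and right two-segment buckle chains hang from master bones on spine_05, with three large spine capsules forming the torso.
# Chain joints allow a narrow single-axis swing with zero twist, while master and spine joints rotate freely.
# The end-segment capsules differ in size between left and right.
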PhysicsAsset /Game/Characters/Player/Female/Medium/Bodies/F_MED_Prison_Break/Meshes/F_MED_Prison_Break_TopBuckles_Front_Physics
bodies {
 dyn_buckleFront_l_01 [{shapeType=Capsule, length=2, radius=2, shapeName=dyn_buckleFront_l_01_capsule, center=<2 -1.4 0>, rotation=<rot 90 0 0>}]
 dyn_buckleFront_l_02 [{shapeType=Capsule, length=3.2814, radius=3.2284, shapeName=dyn_buckleFront_l_02_capsule, center=<3.7665 -2 0>, rotation=<rot 90 0 0>}]
 dyn_buckleFront_l_MSTR [{shapeType=Capsule, length=0.5, radius=0.5, shapeName=dyn_buckleFront_l_MSTR_capsule}]
 dyn_buckleFront_r_01 [{shapeType=Capsule, length=2, radius=2, shapeName=dyn_buckleFront_r_01_capsule, center=<2 -1.4 0>, rotation=<rot 90 0 0>}]
 dyn_buckleFront_r_02 [{shapeType=Capsule, length=5.7814, radius=1.2284, shapeName=dyn_buckleFront_r_02_capsule, center=<3.7665 -0.5 0>, rotation=<rot 90 0 0>}]
 dyn_buckleFront_r_MSTR [{shapeType=Capsule, length=0.5, radius=0.5, shapeName=dyn_buckleFront_r_MSTR_capsule}]
 spine_03 [{shapeType=Capsule, length=4.6971, radius=8.8962, shapeName=spine_04_capsule, center=<0 1.8351 0>}]
 spine_04 [{shapeType=Capsule, length=5.9644, radius=9.053, shapeName=spine_04_capsule, center=<0 1.8351 0>}]
 spine_05 [{shapeType=Capsule, length=8.1075, radius=9.1748, shapeName=spine_05_capsule}]
}
constraints [
 {bone1Name=dyn_buckleFront_l_01, bone2Name=dyn_buckleFront_l_MSTR, swing1LimitDegrees=20, swing2LimitDegrees=0, twistLimitDegrees=0}
 {bone1Name=dyn_buckleFront_l_02, bone2Name=dyn_buckleFront_l_01, swing1LimitDegrees=20, swing2LimitDegrees=0, twistLimitDegrees=0}
 {bone1Name=dyn_buckleFront_r_01, bone2Name=dyn_buckleFront_r_MSTR, swing1LimitDegrees=20, swing2LimitDegrees=0, twistLimitDegrees=0}
 {bone1Name=dyn_buckleFront_r_02, bone2Name=dyn_buckleFront_r_01, swing1LimitDegrees=20, swing2LimitDegrees=0, twistLimitDegrees=0}
 {
  bone1Name=dyn_buckleFront_l_MSTR
  bone2Name=spine_05
  swing1LimitDegrees=45
  swing1Motion=Free
  swing2LimitDegrees=45
  swing2Motion=Free
  twistLimitDegrees=45
  twistMotion=Free
 }
 {
  bone1Name=dyn_buckleFront_r_MSTR
  bone2Name=spine_05
  swing1LimitDegrees=45
  swing1Motion=Free
  swing2LimitDegrees=45
  swing2Motion=Free
  twistLimitDegrees=45
  twistMotion=Free
 }
 {
  bone1Name=spine_05
  bone2Name=spine_04
  swing1LimitDegrees=45
  swing1Motion=Free
  swing2LimitDegrees=45
  swing2Motion=Free
  twistLimitDegrees=45
  twistMotion=Free
 }
 {
  bone1Name=spine_04
  bone2Name=spine_03
  swing1LimitDegrees=45
  swing1Motion=Free
  swing2LimitDegrees=45
  swing2Motion=Free
  twistLimitDegrees=45
  twistMotion=Free
 }
]
