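# Physics asset for the back buckles on the top of the medium female Prison Break body, simulating left and right dyn_buckleBack bone chains.
# Each side is a four-link capsule chain hung from a MSTR body on spine_05, with tight swing limits, zero twist and single-axis swing on the last two links.
# Pelvis, spine and upper-arm capsules form the surrounding body shape, joined with free angular motion.
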PhysicsAsset /Game/Characters/Player/Female/Medium/Bodies/F_MED_Prison_Break/Meshes/F_MED_Prison_Break_TopBuckles_Physics
bodies {
 dyn_buckleBack_l_01 [{shapeType=Capsule, length=6.9031, radius=2, shapeName=dyn_buckleBack_r_01_capsule, center=<3.8259 2 2.9>}]
 dyn_buckleBack_l_02 [{shapeType=Capsule, length=4.922, radius=2, shapeName=dyn_buckleBack_l_02_capsule, center=<2 1.3 0>}]
 dyn_buckleBack_l_03 [{shapeType=Capsule, length=2, radius=2, shapeName=dyn_buckleBack_l_03_capsule, center=<1.2228 1.3 0>}]
 dyn_buckleBack_l_04 [{shapeType=Capsule, length=0.0141, radius=2, shapeName=dyn_buckleBack_l_04_capsule, center=<2 1.3 0>, rotation=<rot 90 0 0>}]
 dyn_buckleBack_l_MSTR [{shapeType=Capsule, length=0.5, radius=0.5, shapeName=dyn_buckleBack_l_MSTR_capsule}]
 dyn_buckleBack_r_01 [{shapeType=Capsule, length=6.9031, radius=2, shapeName=dyn_buckleBack_r_01_capsule, center=<3.8259 2 -2.9>}]
 dyn_buckleBack_r_02 [{shapeType=Capsule, length=4.922, radius=2, shapeName=dyn_buckleBack_r_02_capsule, center=<2 1.3 0>}]
 dyn_buckleBack_r_03 [{shapeType=Capsule, length=1.4164, radius=2.5836, shapeName=dyn_buckleBack_r_03_capsule, center=<1.2228 1.3 0.5759>}]
 dyn_buckleBack_r_04 [{shapeType=Capsule, length=0, radius=2.5836, shapeName=dyn_buckleBack_r_04_capsule, center=<1.4241 1.3 0>, rotation=<rot 90 0 0>}]
 dyn_buckleBack_r_MSTR [{shapeType=Capsule, length=0.5, radius=0.5, shapeName=dyn_buckleBack_r_MSTR_capsule}]
 pelvis [{shapeType=Capsule, length=7.1147, radius=11.8033, shapeName=pelvis_capsule, center=<0.0122 -1.776 -0.059>, rotation=<rot 1.4768 -90 -0.1988>}]
 spine_01 [{
   shapeType=Capsule
   length=4.0576
   radius=9.5098
   shapeName=spine_02_capsule
   center=<2.6044 0.8867 -0.0074>
   rotation=<rot -0.2486 -90 0.368>
  }]
 spine_03 [{shapeType=Capsule, length=4.6971, radius=8.8962, shapeName=spine_04_capsule, center=<0 1.8351 0>}]
 spine_04 [{shapeType=Capsule, length=5.9644, radius=9.053, shapeName=spine_04_capsule, center=<0 1.8351 0>}]
 spine_05 [{shapeType=Capsule, length=8.1075, radius=9.1748, shapeName=spine_05_capsule}]
 upperarm_l [{shapeType=Capsule, length=27.6614, radius=5.6539, shapeName=upperarm_l_capsule, center=<13.0368 -1.4106 0.9085>, rotation=<rot 90 0 0>}]
 upperarm_r [{shapeType=Capsule, length=27.6614, radius=5.6539, shapeName=upperarm_l_capsule, center=<-13.0368 1.4106 -0.9085>, rotation=<rot 90 0 180>}]
}
constraints [
 {bone1Name=dyn_buckleBack_l_01, bone2Name=dyn_buckleBack_l_MSTR, swing1LimitDegrees=15, swing2LimitDegrees=5, twistLimitDegrees=0}
 {bone1Name=dyn_buckleBack_l_02, bone2Name=dyn_buckleBack_l_01, swing1LimitDegrees=10, swing2LimitDegrees=5, twistLimitDegrees=0}
 {bone1Name=dyn_buckleBack_l_03, bone2Name=dyn_buckleBack_l_02, swing1LimitDegrees=20, swing2LimitDegrees=0, twistLimitDegrees=0}
 {bone1Name=dyn_buckleBack_l_04, bone2Name=dyn_buckleBack_l_03, swing1LimitDegrees=20, swing2LimitDegrees=0, twistLimitDegrees=0}
 {bone1Name=dyn_buckleBack_r_01, bone2Name=dyn_buckleBack_r_MSTR, swing1LimitDegrees=20, swing2LimitDegrees=5, twistLimitDegrees=0}
 {bone1Name=dyn_buckleBack_r_02, bone2Name=dyn_buckleBack_r_01, swing1LimitDegrees=20, swing2LimitDegrees=5, twistLimitDegrees=0}
 {bone1Name=dyn_buckleBack_r_03, bone2Name=dyn_buckleBack_r_02, swing1LimitDegrees=20, swing2LimitDegrees=0, twistLimitDegrees=0}
 {bone1Name=dyn_buckleBack_r_04, bone2Name=dyn_buckleBack_r_03, swing1LimitDegrees=20, swing2LimitDegrees=0, twistLimitDegrees=0}
 {
  bone1Name=dyn_buckleBack_l_MSTR
  bone2Name=spine_05
  swing1LimitDegrees=45
  swing1Motion=Free
  swing2LimitDegrees=45
  swing2Motion=Free
  twistLimitDegrees=45
  twistMotion=Free
 }
 {
  bone1Name=dyn_buckleBack_r_MSTR
  bone2Name=spine_05
  swing1LimitDegrees=45
  swing1Motion=Free
  swing2LimitDegrees=45
  swing2Motion=Free
  twistLimitDegrees=45
  twistMotion=Free
 }
 {
  bone1Name=spine_05
  bone2Name=spine_04
  swing1LimitDegrees=45
  swing1Motion=Free
  swing2LimitDegrees=45
  swing2Motion=Free
  twistLimitDegrees=45
  twistMotion=Free
 }
 {
  bone1Name=spine_04
  bone2Name=spine_03
  swing1LimitDegrees=45
  swing1Motion=Free
  swing2LimitDegrees=45
  swing2Motion=Free
  twistLimitDegrees=45
  twistMotion=Free
 }
 {
  bone1Name=spine_03
  bone2Name=spine_01
  swing1LimitDegrees=45
  swing1Motion=Free
  swing2LimitDegrees=45
  swing2Motion=Free
  twistLimitDegrees=45
  twistMotion=Free
 }
 {
  bone1Name=spine_01
  bone2Name=pelvis
  swing1LimitDegrees=45
  swing1Motion=Free
  swing2LimitDegrees=45
  swing2Motion=Free
  twistLimitDegrees=45
  twistMotion=Free
 }
 {
  bone1Name=upperarm_l
  bone2Name=spine_05
  swing1LimitDegrees=45
  swing1Motion=Free
  swing2LimitDegrees=45
  swing2Motion=Free
  twistLimitDegrees=45
  twistMotion=Free
 }
 {
  bone1Name=upperarm_r
  bone2Name=spine_05
  swing1LimitDegrees=45
  swing1Motion=Free
  swing2LimitDegrees=45
  swing2Motion=Free
  twistLimitDegrees=45
  twistMotion=Free
 }
]
