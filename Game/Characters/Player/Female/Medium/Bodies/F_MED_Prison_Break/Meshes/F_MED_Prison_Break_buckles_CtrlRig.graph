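# Control Rig for the female medium Prison Break body that applies pose-driven corrective offsets to dangling cloth bones.
# A spherical pose reader on the left back hip strap weights a Modify Transforms offset on the left back cloth bone.
# A second chain uses thigh pose readers for the front loincloth bones, but is not connected to Begin Execution.
# All rotation offsets are zero, and both left loincloth modifiers also have zero translation.
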ControlRigBlueprint /Game/Characters/Player/Female/Medium/Bodies/F_MED_Prison_Break/Meshes/F_MED_Prison_Break_buckles_CtrlRig
graphs [
 RigVMFunctionLibrary {links [], nodes []}
 RigVMModel {nodes [
   RigUnit_BeginExecution = &RigUnit_BeginExecution()
   n1 = &RigUnit_SphericalPoseReader(
    ActiveRegionScaleFactors=(PositiveWidth=0.1,NegativeWidth=1,PositiveHeight=1,NegativeHeight=1)
    ActiveRegionSize=0.2
    Debug=(bDrawDebug=true,DebugScale=25,DebugSegments=20,DebugThickness=0.25)
    DriverAxis=(X=1,Y=0,Z=0)
    DriverItem=(Type=Bone,Name="dyn_strapHipBack_l_01")
    FalloffRegionScaleFactors=(PositiveWidth=1,NegativeWidth=1,PositiveHeight=1,NegativeHeight=1)
    FalloffSize=0.25
    OptionalParentItem=(Type=Bone)
    RotationOffset=(X=-20,Y=-50,Z=0)
    ExecutePin=@RigUnit_BeginExecution.ExecutePin
   )
   > &RigUnit_ModifyTransforms(
    ItemToModify=((Item=(Type=Bone,Name="dyn_clothBack_l_01"),Transform=(Rotation=(X=0,Y=0,Z=0,W=1),Translation=(X=0,Y=-10,Z=-5),Scale3D=(X=1,Y=1,Z=1))))
    ItemToModify.0.Transform.Rotation=&RigVMFunction_MathQuaternionFromEuler(Euler=(X=0,Y=0,Z=0)).Result
    Weight=@n1.OutputParam
   )
   n2 = &RigUnit_SphericalPoseReader(
    ActiveRegionScaleFactors=(PositiveWidth=1,NegativeWidth=0.1,PositiveHeight=1,NegativeHeight=1)
    ActiveRegionSize=0.2
    Debug=(bDrawDebug=true,DebugScale=25,DebugSegments=20,DebugThickness=0.25)
    DriverAxis=(X=-1,Y=0,Z=0)
    DriverItem=(Type=Bone,Name="thigh_l")
    ExecutePin=()
    FalloffRegionScaleFactors=(PositiveWidth=1,NegativeWidth=1,PositiveHeight=1,NegativeHeight=1)
    FalloffSize=0.2
    OptionalParentItem=(Type=Bone)
    RotationOffset=(X=90,Y=0,Z=40)
   )
   > &RigUnit_ModifyTransforms(
    ItemToModify=((Item=(Type=Bone,Name="dyn_loinClothFront_l_01"),Transform=(Rotation=(X=0,Y=0,Z=0,W=1),Translation=(X=0,Y=0,Z=0),Scale3D=(X=1,Y=1,Z=1))))
    ItemToModify.0.Transform.Rotation=&RigVMFunction_MathQuaternionFromEuler(Euler=(X=0,Y=0,Z=0)).Result
    Weight=@n2.OutputParam
   )
   > n3 = &RigUnit_SphericalPoseReader(
    ActiveRegionScaleFactors=(PositiveWidth=1,NegativeWidth=0.1,PositiveHeight=1,NegativeHeight=1)
    ActiveRegionSize=0.2
    Debug=(bDrawDebug=false,DebugScale=25,DebugSegments=20,DebugThickness=0.25)
    DriverAxis=(X=1,Y=0,Z=0)
    DriverItem=(Type=Bone,Name="thigh_r")
    FalloffRegionScaleFactors=(PositiveWidth=1,NegativeWidth=1,PositiveHeight=1,NegativeHeight=1)
    FalloffSize=0.2
    OptionalParentItem=(Type=Bone)
    RotationOffset=(X=-120,Y=160,Z=60)
   )
   > &RigUnit_ModifyTransforms(
    ItemToModify=((Item=(Type=Bone,Name="dyn_loinClothFront_r_01"),Transform=(Rotation=(X=0,Y=0,Z=0,W=1),Translation=(X=0,Y=-10,Z=0),Scale3D=(X=1,Y=1,Z=1))))
    ItemToModify.0.Transform.Rotation=&RigVMFunction_MathQuaternionFromEuler(Euler=(X=0,Y=0,Z=0)).Result
    Weight=@n3.OutputParam
   )
   > n4 = &RigUnit_SphericalPoseReader(
    ActiveRegionScaleFactors=(PositiveWidth=0.5,NegativeWidth=3.5,PositiveHeight=3,NegativeHeight=3)
    ActiveRegionSize=0.04
    Debug=(bDrawDebug=true,DebugScale=25,DebugSegments=20,DebugThickness=0.25)
    DriverAxis=(X=1,Y=0,Z=0)
    DriverItem=(Type=Bone,Name="thigh_l")
    FalloffRegionScaleFactors=(PositiveWidth=1,NegativeWidth=1,PositiveHeight=1,NegativeHeight=1)
    FalloffSize=0.2
    FlipHeightScaling=false
    FlipWidthScaling=true
    OptionalParentItem=(Type=Bone)
    RotationOffset=(X=0,Y=70,Z=-20)
   )
   > &RigUnit_ModifyTransforms(
    ItemToModify=((Item=(Type=Bone,Name="dyn_loinClothFront_l_01"),Transform=(Rotation=(X=0,Y=0,Z=0,W=1),Translation=(X=0,Y=0,Z=0),Scale3D=(X=1,Y=1,Z=1))))
    ItemToModify.0.Transform.Rotation=&RigVMFunction_MathQuaternionFromEuler(Euler=(X=0,Y=0,Z=0)).Result
    Weight=@n4.OutputParam
   )
  ]}
]
hierarchy [
 root {loc=<0 0 0>, rot=<0 0 0 1>}
 pelvis {parent=root, loc=<0 -0.5301 98.0571>, rot=<0.0127 0.707 -0.0127 -0.707>}
 spine_01 {parent=pelvis, loc=<6.1208 -0.9328 0>, rot=<0 0 0.034 -0.9994>}
 spine_02 {parent=spine_01, loc=<6.5996 -0.2095 0>, rot=<0 0 0.0238 -0.9997>}
 spine_03 {parent=spine_02, loc=<6.4914 -0.1237 0>, rot=<0 0 0.0149 -0.9999>}
 spine_04 {parent=spine_03, loc=<6.9598 0.0794 0>, rot=<0 0 -0.0132 -0.9999>}
 spine_05 {parent=spine_04, loc=<6.9987 0 0>, rot=<0 0 -0.0015 -1>}
 clavicle_l {parent=spine_05, loc=<5.4831 -0.1053 -3.2946>, rot=<0.0465 -0.7289 -0.009 -0.6829>}
 upperarm_l {parent=clavicle_l, loc=<9.8658 -1.2136 -0.4105>, rot=<0.0164 -0.3245 -0.0269 -0.9454>}
 lowerarm_l {parent=upperarm_l, loc=<26.465 -0.8345 0.259>, rot=<0.0036 0.0161 -0.2298 -0.9731>}
 hand_l {parent=lowerarm_l, loc=<22.7355 -0.3203 -0.6105>, rot=<-0.6784 -0.0223 0.0735 -0.7307>}
 thumb_01_l {parent=hand_l, loc=<1.3118 -0.3191 -2.459>, rot=<0.6546 -0.4042 0.0664 -0.6354>}
 thumb_02_l {parent=thumb_01_l, loc=<5.9336 -0.0448 0.0401>, rot=<-0.0037 0.0213 0.298 -0.9543>}
 thumb_03_l {parent=thumb_02_l, loc=<2.2223 0.3682 -0.0907>, rot=<0.0118 -0.013 0.0066 -0.9998>}
 index_metacarpal_l {parent=hand_l, loc=<1.6535 0.0183 -2.96>, rot=<0.0158 -0.0124 -0.0173 -0.9996>}
 index_01_l {parent=index_metacarpal_l, loc=<8.3008 0.0357 -0.0524>, rot=<-0.0044 0.0485 0.0924 -0.9945>}
 index_02_l {parent=index_01_l, loc=<3.3587 -0.3004 0.0452>, rot=<-0.002 0.0027 0.3225 -0.9466>}
 index_03_l {parent=index_02_l, loc=<2.5578 -0.123 -0.0252>, rot=<-0.002 0.0067 0.2558 -0.9667>}
 middle_metacarpal_l {parent=hand_l, loc=<2.1856 -0.1725 -0.5984>, rot=<0.006 0.0168 -0.0329 -0.9993>}
 middle_01_l {parent=middle_metacarpal_l, loc=<7.3275 -0.2447 0.0129>, rot=<-0.0208 0.0042 0.0858 -0.9961>}
 middle_02_l {parent=middle_01_l, loc=<4.0123 0.1401 0.1175>, rot=<-0.0218 0.0258 0.3512 -0.9357>}
 middle_03_l {parent=middle_02_l, loc=<2.6077 0.0534 0.1373>, rot=<0.0082 0.0188 0.1708 -0.9851>}
 ring_metacarpal_l {parent=hand_l, loc=<2.2074 -0.3625 1.3638>, rot=<-0.0284 0.0722 -0.0458 -0.9959>}
 ring_01_l {parent=ring_metacarpal_l, loc=<7.0295 -0.6197 -0.1446>, rot=<-0.0489 -0.0252 0.1892 -0.9804>}
 ring_02_l {parent=ring_01_l, loc=<3.6174 0.4647 -0.1027>, rot=<0 0 0.2806 -0.9598>}
 ring_03_l {parent=ring_02_l, loc=<2.4152 0.114 0.0199>, rot=<0 0 0.1189 -0.9929>}
 pinky_metacarpal_l {parent=hand_l, loc=<1.7616 -0.9753 3.3814>, rot=<-0.0397 0.0814 -0.069 -0.9935>}
 pinky_01_l {parent=pinky_metacarpal_l, loc=<7.1709 -0.4831 -0.1094>, rot=<-0.0654 -0.0035 0.2242 -0.9723>}
 pinky_02_l {parent=pinky_01_l, loc=<2.8274 0.0893 -0.1043>, rot=<0.0037 0.0011 0.2854 -0.9584>}
 pinky_03_l {parent=pinky_02_l, loc=<2.2046 0.0307 -0.0078>, rot=<-0.0002 0.0098 0.1445 -0.9895>}
 deform_wrist_dn_base_l {parent=hand_l, loc=<0 -1.5326 0>, rot=<0 0 0 -1>}
 deform_wrist_dn_l {parent=deform_wrist_dn_base_l, loc=<0 0 0>, rot=<0 0 0 1>}
 deform_wrist_up_base_l {parent=hand_l, loc=<0 2.3009 0>, rot=<0 0 0 -1>}
 deform_wrist_up_l {parent=deform_wrist_up_base_l, loc=<0 0 0>, rot=<0 0 0 1>}
 lowerarm_twist_02_l {parent=lowerarm_l, loc=<10.6707 0 0>, rot=<0 0 0 1>}
 lowerarm_twist_01_l {parent=lowerarm_l, loc=<18.316 0 0>, rot=<0 0 0 1>}
 deform_elbow_in_base_l {parent=lowerarm_l, loc=<0.1834 3.6062 -0.2572>, rot=<0 0 0 -1>}
 deform_elbow_in_l {parent=deform_elbow_in_base_l, loc=<0 0 0>, rot=<0 0 0 1>}
 deform_elbow_out_base_l {parent=lowerarm_l, loc=<0 -4 0>, rot=<0 0 0 1>}
 deform_elbow_out_l {parent=deform_elbow_out_base_l, loc=<0 0 0>, rot=<0 0 0 1>}
 upperarm_twist_01_l {parent=upperarm_l, loc=<7.2135 -0.1134 0>, rot=<0 0 0 1>}
 deform_upperarm_bk_base_l {parent=upperarm_twist_01_l, loc=<-6.4053 -4.3746 0.314>, rot=<0 0 0 -1>}
 deform_upperarm_bk_l {parent=deform_upperarm_bk_base_l, loc=<0 0 0>, rot=<0 0 0 1>}
 deform_upperarm_fr_base_l {parent=upperarm_twist_01_l, loc=<-5.6312 5.092 0.3354>, rot=<0 0 0 -1>}
 deform_upperarm_fr_l {parent=deform_upperarm_fr_base_l, loc=<0 0 0>, rot=<0 0 0 1>}
 deform_upperarm_up_base_l {parent=upperarm_twist_01_l, loc=<-8.891 0.4825 3.9517>, rot=<0 0 0 -1>}
 deform_upperarm_up_l {parent=deform_upperarm_up_base_l, loc=<0 0 0>, rot=<0 0 0 1>}
 upperarm_twist_02_l {parent=upperarm_l, loc=<19.7998 -0.1134 0>, rot=<0 0 0 1>}
 attach_cape {parent=spine_05, loc=<9.3216 -8.4501 0>, rot=<0.0616 -0.7044 -0.0616 -0.7044>}
 attach_backpack {parent=spine_05, loc=<0 -8.35 0>, rot=<-0.0093 -0.707 0.0093 -0.707>}
 neck_01 {parent=spine_05, loc=<11.3138 0.1529 0>, rot=<0 0 -0.2003 -0.9797>}
 neck_02 {parent=neck_01, loc=<4.5627 0 0>, rot=<0 0 0.0504 -0.9987>}
 head {parent=neck_02, loc=<5.3914 0 0>, rot=<0 0 0.115 -0.9934>}
 clavicle_r {parent=spine_05, loc=<5.4831 -0.1053 3.2946>, rot=<0.7289 0.0465 -0.6829 0.009>}
 upperarm_r {parent=clavicle_r, loc=<-9.8658 1.2136 0.4105>, rot=<0.0164 -0.3245 -0.0269 -0.9454>}
 lowerarm_r {parent=upperarm_r, loc=<-26.465 0.8345 -0.259>, rot=<0.0036 0.0161 -0.2298 -0.9731>}
 hand_r {parent=lowerarm_r, loc=<-22.7355 0.3203 0.6106>, rot=<-0.6784 -0.0223 0.0735 -0.7307>}
 thumb_01_r {parent=hand_r, loc=<-1.3118 0.3191 2.459>, rot=<0.6546 -0.4042 0.0664 -0.6354>}
 thumb_02_r {parent=thumb_01_r, loc=<-5.9336 0.0448 -0.0401>, rot=<-0.0037 0.0213 0.298 -0.9543>}
 thumb_03_r {parent=thumb_02_r, loc=<-2.2223 -0.3682 0.0907>, rot=<0.0118 -0.013 0.0066 -0.9998>}
 index_metacarpal_r {parent=hand_r, loc=<-1.6535 -0.0183 2.96>, rot=<0.0158 -0.0124 -0.0173 -0.9996>}
 index_01_r {parent=index_metacarpal_r, loc=<-8.3008 -0.0357 0.0524>, rot=<-0.0044 0.0485 0.0924 -0.9945>}
 index_02_r {parent=index_01_r, loc=<-3.3587 0.3004 -0.0452>, rot=<-0.002 0.0027 0.3225 -0.9466>}
 index_03_r {parent=index_02_r, loc=<-2.5578 0.123 0.0252>, rot=<-0.002 0.0067 0.2558 -0.9667>}
 middle_metacarpal_r {parent=hand_r, loc=<-2.1856 0.1725 0.5984>, rot=<0.006 0.0168 -0.0329 -0.9993>}
 middle_01_r {parent=middle_metacarpal_r, loc=<-7.3275 0.2447 -0.0129>, rot=<-0.0208 0.0042 0.0858 -0.9961>}
 middle_02_r {parent=middle_01_r, loc=<-4.0123 -0.1401 -0.1175>, rot=<-0.0218 0.0258 0.3512 -0.9357>}
 middle_03_r {parent=middle_02_r, loc=<-2.6077 -0.0534 -0.1373>, rot=<0.0082 0.0188 0.1708 -0.9851>}
 ring_metacarpal_r {parent=hand_r, loc=<-2.2074 0.3625 -1.3638>, rot=<-0.0284 0.0722 -0.0458 -0.9959>}
 ring_01_r {parent=ring_metacarpal_r, loc=<-7.0295 0.6197 0.1446>, rot=<-0.0489 -0.0252 0.1892 -0.9804>}
 ring_02_r {parent=ring_01_r, loc=<-3.6174 -0.4647 0.1027>, rot=<0 0 0.2806 -0.9598>}
 ring_03_r {parent=ring_02_r, loc=<-2.4152 -0.114 -0.0199>, rot=<0 0 0.1189 -0.9929>}
 pinky_metacarpal_r {parent=hand_r, loc=<-1.7616 0.9753 -3.3814>, rot=<-0.0397 0.0814 -0.069 -0.9935>}
 pinky_01_r {parent=pinky_metacarpal_r, loc=<-7.1709 0.4831 0.1094>, rot=<-0.0654 -0.0035 0.2242 -0.9723>}
 pinky_02_r {parent=pinky_01_r, loc=<-2.8274 -0.0893 0.1043>, rot=<0.0037 0.0011 0.2854 -0.9584>}
 pinky_03_r {parent=pinky_02_r, loc=<-2.2046 -0.0307 0.0078>, rot=<-0.0002 0.0098 0.1445 -0.9895>}
 deform_wrist_dn_base_r {parent=hand_r, loc=<-0.0001 1.5328 0>, rot=<0 0 0 -1>}
 deform_wrist_dn_r {parent=deform_wrist_dn_base_r, loc=<0 0 0>, rot=<0 0 0 1>}
 deform_wrist_up_base_r {parent=hand_r, loc=<-0.0002 -2.3007 -0.0001>, rot=<0 0 0 -1>}
 deform_wrist_up_r {parent=deform_wrist_up_base_r, loc=<0 0 0>, rot=<0 0 0 1>}
 lowerarm_twist_02_r {parent=lowerarm_r, loc=<-10.6707 0 0>, rot=<0 0 0 -1>}
 lowerarm_twist_01_r {parent=lowerarm_r, loc=<-18.316 0 0>, rot=<0 0 0 -1>}
 deform_elbow_in_base_r {parent=lowerarm_r, loc=<-0.1834 -3.6062 0.2572>, rot=<0 0 0 -1>}
 deform_elbow_in_r {parent=deform_elbow_in_base_r, loc=<0 0 0>, rot=<0 0 0 1>}
 deform_elbow_out_base_r {parent=lowerarm_r, loc=<0 4 -0.0001>, rot=<0 0 0 -1>}
 deform_elbow_out_r {parent=deform_elbow_out_base_r, loc=<0 0 0>, rot=<0 0 0 1>}
 upperarm_twist_01_r {parent=upperarm_r, loc=<-7.2135 0.1134 0>, rot=<0 0 0 -1>}
 deform_upperarm_bk_base_r {parent=upperarm_twist_01_r, loc=<6.4053 4.3746 -0.314>, rot=<0 0 0 -1>}
 deform_upperarm_bk_r {parent=deform_upperarm_bk_base_r, loc=<0 0 0>, rot=<0 0 0 1>}
 deform_upperarm_fr_base_r {parent=upperarm_twist_01_r, loc=<5.6312 -5.092 -0.3354>, rot=<0 0 0 -1>}
 deform_upperarm_fr_r {parent=deform_upperarm_fr_base_r, loc=<0 0 0>, rot=<0 0 0 1>}
 deform_upperarm_up_base_r {parent=upperarm_twist_01_r, loc=<8.891 -0.4825 -3.9517>, rot=<0 0 0 -1>}
 deform_upperarm_up_r {parent=deform_upperarm_up_base_r, loc=<0 0 0>, rot=<0 0 0 1>}
 upperarm_twist_02_r {parent=upperarm_r, loc=<-19.7998 0.1134 0>, rot=<0 0 0 -1>}
 dyn_buckleFront_l_MSTR {parent=spine_05, loc=<-0.1447 5.4821 -13.2666>, rot=<0.0103 0.2631 0.9584 0.1101>}
 dyn_buckleFront_l_01 {parent=dyn_buckleFront_l_MSTR, loc=<0 0 0>, rot=<0 0 0 1>}
 dyn_buckleFront_l_02 {parent=dyn_buckleFront_l_01, loc=<4.0787 0 0>, rot=<-0.0024 -0.029 0.0697 -0.9971>}
 dyn_buckleBack_l_MSTR {parent=spine_05, loc=<1.3944 -8.3876 -7.0639>, rot=<0.0024 -0.0554 0.9977 -0.0386>}
 dyn_buckleBack_l_01 {parent=dyn_buckleBack_l_MSTR, loc=<0 0 0>, rot=<0 0 0 1>}
 dyn_buckleBack_l_02 {parent=dyn_buckleBack_l_01, loc=<8.6387 0 0>, rot=<0 0 0 1>}
 dyn_buckleBack_l_03 {parent=dyn_buckleBack_l_02, loc=<7.0027 -0.0006 -0.0053>, rot=<0 0 0 1>}
 dyn_buckleBack_l_04 {parent=dyn_buckleBack_l_03, loc=<2.3309 0 0>, rot=<0 0 0 1>}
 dyn_buckleBack_r_MSTR {parent=spine_05, loc=<1.4272 -8.385 7.0511>, rot=<-0.0019 0.0558 0.9977 -0.0386>}
 dyn_buckleBack_r_01 {parent=dyn_buckleBack_r_MSTR, loc=<0 0 0>, rot=<0 0 0 1>}
 dyn_buckleBack_r_02 {parent=dyn_buckleBack_r_01, loc=<8.6715 0 0>, rot=<0 0 0 1>}
 dyn_buckleBack_r_03 {parent=dyn_buckleBack_r_02, loc=<6.9694 0 0>, rot=<0 0 0 1>}
 dyn_buckleBack_r_04 {parent=dyn_buckleBack_r_03, loc=<2.3644 0 0>, rot=<0 0 0 1>}
 dyn_buckleFront_r_MSTR {parent=spine_05, loc=<-0.1878 5.3259 13.1698>, rot=<-0.0012 -0.2678 0.956 0.1197>}
 dyn_buckleFront_r_01 {parent=dyn_buckleFront_r_MSTR, loc=<0 0 0>, rot=<0 0 0 1>}
 dyn_buckleFront_r_02 {parent=dyn_buckleFront_r_01, loc=<4.0521 0 0>, rot=<0.0189 0.0202 0.0599 -0.9978>}
 deform_pec_aim_base_l {parent=spine_04, loc=<3.2456 9.3309 -1.2237>, rot=<0.0603 -0.4033 0.3318 -0.8507>}
 deform_pec_aim_l {parent=deform_pec_aim_base_l, loc=<0 0 0>, rot=<0 0 0 1>}
 deform_pec_base_l {parent=deform_pec_aim_l, loc=<9.3665 0 0>, rot=<0 0 0 1>}
 deform_pec_l {parent=deform_pec_base_l, loc=<0 0 0>, rot=<0 0 0 1>}
 deform_pec_aim_base_r {parent=spine_04, loc=<3.2452 9.3309 1.2237>, rot=<-0.4033 -0.0603 0.8507 0.3318>}
 deform_pec_aim_r {parent=deform_pec_aim_base_r, loc=<0 0 0>, rot=<0 0 0 1>}
 deform_pec_base_r {parent=deform_pec_aim_r, loc=<-9.3667 -0.0003 -0.0004>, rot=<0 0 0 1>}
 deform_pec_r {parent=deform_pec_base_r, loc=<0 0 0>, rot=<0 0 0 1>}
 thigh_l {parent=pelvis, loc=<-1.0855 -0.2109 -8.7053>, rot=<0.0009 0.0355 0.0155 -0.9993>}
 calf_l {parent=thigh_l, loc=<-42.5284 -0.0019 -0.7297>, rot=<0.0006 -0.0161 -0.0391 -0.9991>}
 foot_l {parent=calf_l, loc=<-46.7401 -2.5014 -0.592>, rot=<0.0427 -0.0238 0.0353 -0.9982>}
 ball_l {parent=foot_l, loc=<-3.5658 10.8066 -0.1544>, rot=<0.0001 -0.0001 -0.7063 -0.7079>}
 calf_twist_02_l {parent=calf_l, loc=<-19.1546 -1.7943 -0.2638>, rot=<0.0237 0.0095 -0.0035 -0.9997>}
 dyn_legStrap_l_01 {parent=calf_twist_02_l, loc=<-0.3615 -2.2424 -5.2428>, rot=<0.2744 -0.6663 0.2448 -0.6488>}
 calf_twist_01_l {parent=calf_l, loc=<-35.8483 -1.9197 -0.5794>, rot=<0.0237 0.0095 -0.0035 -0.9997>}
 dyn_legStrap_l_02 {parent=calf_twist_01_l, loc=<5.5163 -1.0473 -3.5105>, rot=<0.3171 -0.6922 0.2724 -0.5883>}
 deform_knee_in_base_l {parent=calf_l, loc=<0.8031 -5.0231 0.2456>, rot=<0 0 0 -1>}
 deform_knee_in_l {parent=deform_knee_in_base_l, loc=<0 0 0>, rot=<0 0 0 1>}
 deform_knee_out_base_l {parent=calf_l, loc=<0 4 0>, rot=<0 0 0 -1>}
 deform_knee_out_l {parent=deform_knee_out_base_l, loc=<0 0 0>, rot=<0 0 0 1>}
 thigh_twist_01_l {parent=thigh_l, loc=<-27.952 0 0>, rot=<0.0013 0 0 -1>}
 deform_groin_base_l {parent=thigh_l, loc=<0.4801 5.6776 -0.2486>, rot=<0 0 0 -1>}
 deform_groin_l {parent=deform_groin_base_l, loc=<0 0 0>, rot=<0 0 0 1>}
 dyn_strapHipFront_l_MSTR {parent=thigh_l, loc=<-0.7777 4.6517 -2.0645>, rot=<-0.098 -0.9483 0.2453 -0.1757>}
 dyn_strapHipFront_l_01 {parent=dyn_strapHipFront_l_MSTR, loc=<0 0 0>, rot=<0 0 0 1>}
 dyn_strapHipFront_l_02 {parent=dyn_strapHipFront_l_01, loc=<2.9576 -0.0039 0>, rot=<0.0009 0.0296 0.0635 -0.9975>}
 dyn_strapHipFront_l_03 {parent=dyn_strapHipFront_l_02, loc=<5.1273 -0.0114 0.0813>, rot=<-0.0053 0.1035 -0.0596 -0.9928>}
 thigh_r {parent=pelvis, loc=<-1.0855 -0.2109 8.7053>, rot=<0.0355 -0.0009 0.9993 0.0155>}
 calf_r {parent=thigh_r, loc=<42.5284 0.0019 0.7297>, rot=<0.0006 -0.0161 -0.0391 -0.9991>}
 foot_r {parent=calf_r, loc=<46.7401 2.5014 0.592>, rot=<0.0427 -0.0238 0.0353 -0.9982>}
 ball_r {parent=foot_r, loc=<3.5658 -10.8066 0.1544>, rot=<0.0001 -0.0001 -0.7063 -0.7079>}
 calf_twist_02_r {parent=calf_r, loc=<19.1546 1.7943 0.2638>, rot=<0.0237 0.0095 -0.0035 -0.9997>}
 dyn_legStrap_r_01 {parent=calf_twist_02_r, loc=<0.3615 2.2424 5.2428>, rot=<0.264 0.6608 -0.2481 -0.6573>}
 calf_twist_01_r {parent=calf_r, loc=<35.8483 1.9197 0.5794>, rot=<0.0237 0.0095 -0.0035 -0.9997>}
 dyn_legStrap_r_02 {parent=calf_twist_01_r, loc=<-5.5163 1.0473 3.5105>, rot=<0.2854 0.6086 -0.2928 -0.68>}
 deform_knee_in_base_r {parent=calf_r, loc=<-0.8031 5.0231 -0.2457>, rot=<0 0 0 -1>}
 deform_knee_in_r {parent=deform_knee_in_base_r, loc=<0 0 0>, rot=<0 0 0 1>}
 deform_knee_out_base_r {parent=calf_r, loc=<0 -4 0>, rot=<0 0 0 -1>}
 deform_knee_out_r {parent=deform_knee_out_base_r, loc=<0 0 0>, rot=<0 0 0 1>}
 thigh_twist_01_r {parent=thigh_r, loc=<27.952 0 0>, rot=<0.0013 0 0 -1>}
 deform_groin_base_r {parent=thigh_r, loc=<-0.4801 -5.6776 0.2486>, rot=<0 0 0 -1>}
 deform_groin_r {parent=deform_groin_base_r, loc=<0 0 0>, rot=<0 0 0 1>}
 dyn_strapHipFront_r_MSTR {parent=thigh_r, loc=<0.6932 -4.7281 2.0463>, rot=<0.9549 -0.0645 0.1562 0.2441>}
 dyn_strapHipFront_r_01 {parent=dyn_strapHipFront_r_MSTR, loc=<3.0505 -0.0103 -0.1056>, rot=<0 0 0.1119 -0.9937>}
 dyn_strapHipFront_r_02 {parent=dyn_strapHipFront_r_01, loc=<3.9335 0.5126 -1.5112>, rot=<-0.0414 -0.2307 0.2026 -0.9508>}
 dyn_strapHipBack_r_MSTR {parent=thigh_r, loc=<-3.2214 10.2353 -1.9453>, rot=<0.5368 0.2239 0.1839 -0.7924>}
 dyn_strapHipBack_r_01 {parent=dyn_strapHipBack_r_MSTR, loc=<7.7127 0.1372 -0.2089>, rot=<0.0156 -0.1635 0.0936 -0.982>}
 dyn_strapHipBack_r_02 {parent=dyn_strapHipBack_r_01, loc=<7.0107 0 0>, rot=<0 -0.3902 0 -0.9207>}
 deform_glute_base_l {parent=pelvis, loc=<-7.3466 -9.0213 -9.1245>, rot=<0.0172 0.031 -0.4477 -0.8935>}
 deform_glute_l {parent=deform_glute_base_l, loc=<0 0 0>, rot=<0 0 0 1>}
 deform_glute_base_r {parent=pelvis, loc=<-7.3464 -9.0213 9.1245>, rot=<0.031 -0.0172 0.8935 -0.4477>}
 deform_glute_r {parent=deform_glute_base_r, loc=<0 0 0>, rot=<0 0 0 1>}
 dyn_clothBack_MSTR {parent=pelvis, loc=<6.7421 -10.6533 0>, rot=<0.1167 -0.9932 0 0>}
 dyn_clothBack_01 {parent=dyn_clothBack_MSTR, loc=<0 0 0>, rot=<0 0 0 1>}
 dyn_clothBack_02 {parent=dyn_clothBack_01, loc=<10.6657 0 0>, rot=<0 0 -0.1019 -0.9948>}
 dyn_clothBack_03 {parent=dyn_clothBack_02, loc=<9.773 0 0>, rot=<0 0 -0.0051 -1>}
 dyn_clothBack_04 {parent=dyn_clothBack_03, loc=<10.5557 0 0>, rot=<0 0 -0.0277 -0.9996>}
 dyn_belt2_MSTR {parent=dyn_clothBack_03, loc=<1.6658 1.8148 -4.0279>, rot=<-0.0046 -0.0003 0.0172 -0.9998>}
 dyn_belt2_01 {parent=dyn_belt2_MSTR, loc=<0 0 0>, rot=<0 0 0 1>}
 dyn_belt2_02 {parent=dyn_belt2_01, loc=<8.6833 0 0>, rot=<0 0 0.0155 -0.9999>}
 dyn_belt2_03 {parent=dyn_belt2_02, loc=<15.9955 0 0>, rot=<0.0044 0.0006 0.0102 -0.9999>}
 dyn_belt2_04 {parent=dyn_belt2_03, loc=<14.1028 0 0>, rot=<0.0013 0.0002 -0.0058 -1>}
 dyn_belt2_05 {parent=dyn_belt2_04, loc=<8.536 0 0>, rot=<0 0 0 1>}
 dyn_belt2_06 {parent=dyn_belt2_05, loc=<3.0206 0 0>, rot=<0 0 0 1>}
 dyn_clothBack_l_01 {parent=dyn_clothBack_01, loc=<9.4504 1.5066 10.0339>, rot=<0.1424 0.0397 -0.136 -0.9796>}
 dyn_teddyAnchor_MSTR {parent=pelvis, loc=<0.7998 -2.0848 -13.1565>, rot=<0.1568 -0.7011 -0.673 -0.176>}
 dyn_teddyAnchor_01 {parent=dyn_teddyAnchor_MSTR, loc=<0 0 0>, rot=<0.0016 0.0002 0.001 -1>}
 dyn_teddyBody_MSTR {parent=dyn_teddyAnchor_01, loc=<13.761 0.0283 -0.0049>, rot=<-0.0027 0.352 0.9299 -0.1072>}
 dyn_teddyBody_01 {parent=dyn_teddyBody_MSTR, loc=<0 0 0>, rot=<0.065 -0.0303 0.9818 -0.1759>}
 dyn_teddyBody_02 {parent=dyn_teddyBody_01, loc=<5.9476 0 0>, rot=<0 0 0.1786 -0.9839>}
 dyn_teddyBody_03 {parent=dyn_teddyBody_02, loc=<4.9381 0 0>, rot=<0 0 0 1>}
 dyn_teddyLeg_l_01 {parent=dyn_teddyBody_03, loc=<4.9075 -0.4338 -2.2818>, rot=<-0.0156 0.0089 0.1052 -0.9943>}
 dyn_teddyLeg_l_02 {parent=dyn_teddyLeg_l_01, loc=<5.7063 0 0>, rot=<0 0 0 1>}
 dyn_teddyLeg_r_01 {parent=dyn_teddyBody_03, loc=<4.5126 0.0716 3.6945>, rot=<0.1346 0.0963 0.1368 -0.9767>}
 dyn_teddyLeg_r_02 {parent=dyn_teddyLeg_r_01, loc=<5.6829 0 0>, rot=<0 0 0 1>}
 dyn_teddyArm_l_01 {parent=dyn_teddyBody_01, loc=<0.2209 -2.9288 -3.0497>, rot=<-0.2096 0.0427 0.4977 -0.8406>}
 dyn_teddyArm_r_01 {parent=dyn_teddyBody_01, loc=<-5.1281 -2.9039 5.5071>, rot=<0.0025 0.2636 0.5834 -0.7683>}
 dyn_teddyHead_01 {parent=dyn_teddyBody_01, loc=<-2.1581 0.0904 -0.0576>, rot=<-0.0722 0.1171 0.9871 -0.0818>}
 dyn_strapHipBack_l_MSTR {parent=pelvis, loc=<1.4514 -10.4919 -6.1659>, rot=<0.0213 -0.9254 -0.1701 -0.338>}
 dyn_strapHipBack_l_01 {parent=dyn_strapHipBack_l_MSTR, loc=<0 0 0>, rot=<0 0 0 1>}
 dyn_strapHipBack_l_02 {parent=dyn_strapHipBack_l_01, loc=<6.9682 -0.0184 -0.0232>, rot=<0.007 0.0627 -0.06 -0.9962>}
 dyn_loinClothFront_l_01 {parent=pelvis, loc=<-2.947 8.083 -4.35>, rot=<-0.0179 -0.9998 0 0>}
 dyn_loinClothFront_l_02 {parent=dyn_loinClothFront_l_01, loc=<6.1997 0 0>, rot=<0 0 0 1>}
 dyn_loinClothFront_r_01 {parent=pelvis, loc=<-2.947 8.083 4.35>, rot=<-0.0179 -0.9998 0 0>}
 dyn_loinClothFront_r_02 {parent=dyn_loinClothFront_r_01, loc=<6.1997 0 0>, rot=<0 0 0 1>}
]
