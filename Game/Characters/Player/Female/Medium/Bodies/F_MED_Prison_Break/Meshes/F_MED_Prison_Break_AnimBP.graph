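# Post-process style AnimBlueprint for the female Prison Break body part, adding secondary motion to straps, buckles, belt cloth and a teddy bear.
# The copied parent pose runs through a deform Control Rig, cached RigidBody stages, a teddy-strap LookAt, then a buckle Control Rig and belt physics.
# The EventGraph updates gravity override, movement flags, backbling detection that disables top-buckle physics, and velocity-driven WPO material regions.
# Strap Constraint nodes sit disconnected from the output.
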AnimBlueprint /Game/Characters/Player/Female/Medium/Bodies/F_MED_Prison_Break/Meshes/F_MED_Prison_Break_AnimBP
target_skeleton=&/F_MED_Prison_Break_Skeleton
graphs [
 AnimGraph {nodes [
   n1 = "Constraint - dyn_strapHipFront_r_MSTR (PPPPPP)"!(
    ComponentPose=Local To Component()
    Alpha=1
    ConstraintWeights_0=0.5546
    ConstraintWeights_1:float=0.3127
    ConstraintWeights_2:float=0.1231
    ConstraintWeights_3:float=0.0091
    ConstraintWeights_4:float=0.0002
    ConstraintWeights_5:float=0.0002
   ) -> (Pose) {
    alphaBoolBlend {blendOption=Linear}
    boneToModify {boneName=dyn_strapHipFront_r_MSTR}
    constraintSetup [
     {targetBone {boneName=pelvis}}
     {targetBone {boneName=deform_groin_r}}
     {targetBone {boneName=thigh_twist_01_r}}
     {targetBone {boneName=spine_02}}
     {targetBone {boneName=thigh_twist_01_l}}
     {targetBone {boneName=thigh_r}}
    ]
    constraintWeights [1,1,1,1,1,1]
   }
   Component To Local(ComponentPose=@n1)
   "Constraint - dyn_strapHipFront_l_01 (PPPPP)"!(
    ComponentPose
    Alpha=1
    ConstraintWeights_0=0.5358
    ConstraintWeights_1:float=0.2691
    ConstraintWeights_2:float=0.1911
    ConstraintWeights_3:float=0.0036
    ConstraintWeights_4:float=0.0004
   ) -> (Pose) {
    alphaBoolBlend {blendOption=Linear}
    boneToModify {boneName=dyn_strapHipFront_l_01}
    constraintSetup [
     {targetBone {boneName=pelvis}}
     {targetBone {boneName=deform_groin_l}}
     {targetBone {boneName=thigh_twist_01_l}}
     {targetBone {boneName=spine_02}}
     {targetBone {boneName=thigh_l}}
    ]
    constraintWeights [1,1,1,1,1]
   }
   "Transform (Modify) Bone"(Alpha=1, Translation="0,0,0", Rotation="0,0,0", Scale="1,1,1") {alphaBoolBlend {blendOption=Linear}, boneToModify=~}
   n2 = Control Rig(Source=Use cached pose 'frontLoinClothCache'()) {
    alphaBoolBlend {blendOption=Linear}
    controlRigAssetReference {blueprintRigClass=&/F_MED_Prison_Break_buckles_CtrlRig.F_MED_Prison_Break_buckles_CtrlRig_C}
   }
   n3 = RigidBody!(ComponentPose=Local To Component(LocalPose=@n2), Alpha, OverrideWorldGravity:Vector=Get Gravity Override(), ExternalForce) -> (Pose) {
    bClampLinearTranslationLimitToRefPose=true
    bOverrideWorldGravity=true
    overridePhysicsAsset=&/F_MED_Prison_Break_BottomClothBelt2_Physics
    simulationSpace=BaseBoneSpace
    alphaBoolBlend {blendOption=Linear}
    baseBoneRef {boneName=pelvis}
    componentAppliedLinearAccClamp=<6000 6000 6000>
    componentLinearAccScale=<0.2 0.4 0.2>
    componentLinearVelScale=<0.2 0.4 0.2>
   }
   Output Pose(Result=Component To Local(ComponentPose=@n3)) {layerGroup=None}
   n4 = Control Rig(Source=Copy Pose From Mesh() {bUseAttachedParent=true}) {
    alphaBoolBlend {blendOption=Linear}
    controlRigAssetReference {blueprintRigClass=&/F_MED_DeformRig_PrisonBreakMaster_ControlRig.F_MED_DeformRig_PrisonBreakMaster_ControlRig_C}
   }
   Save cached pose 'CRCache'(Pose=@n4) {cacheName=CRCache}
   n5 = Use cached pose 'teddyCache'()
   Save cached pose 'frontLoinClothCache'(Pose=Component To Local(ComponentPose=Local To Component(LocalPose=@n5))) {cacheName=frontLoinClothCache}
   n6 = RigidBody!(
    ComponentPose=Local To Component(LocalPose=Use cached pose 'CRCache'())
    bAlphaBoolEnabled:bool=NOT Boolean(A=Get Is Wearing Backbling())
    OverrideWorldGravity=Get Gravity Override()
    ExternalForce
   ) -> (Pose) {
    alphaInputType=Bool
    bClampLinearTranslationLimitToRefPose=true
    bOverrideWorldGravity=true
    overridePhysicsAsset=&/F_MED_Prison_Break_TopBuckles_Physics
    simulationSpace=BaseBoneSpace
    alphaBoolBlend {blendOption=Linear}
    baseBoneRef {boneName=spine_05}
    componentAppliedLinearAccClamp=<6000 6000 6000>
    componentLinearAccScale=<0.5 0.3 0.3>
    componentLinearVelScale=<0.5 0.3 0.3>
   }
   n7 = RigidBody!(ComponentPose=@n6, Alpha:float, OverrideWorldGravity=Get Gravity Override(), ExternalForce) -> (Pose) {
    bClampLinearTranslationLimitToRefPose=true
    bOverrideWorldGravity=true
    overridePhysicsAsset=&/F_MED_Prison_Break_TopBuckles_Front_Physics
    simulationSpace=BaseBoneSpace
    alphaBoolBlend {blendOption=Linear}
    baseBoneRef {boneName=spine_05}
    componentAppliedLinearAccClamp=<6000 6000 6000>
    componentLinearAccScale=<0.5 0.3 0.3>
    componentLinearVelScale=<0.5 0.3 0.3>
   }
   Save cached pose 'bucklesCache'(Pose=Component To Local(ComponentPose=@n7)) {cacheName=bucklesCache}
   n8 = RigidBody(ComponentPose=Local To Component(LocalPose=Use cached pose 'bucklesCache'()), OverrideWorldGravity=Get Gravity Override()) {
    bClampLinearTranslationLimitToRefPose=true
    bOverrideWorldGravity=true
    overridePhysicsAsset=&/F_MED_Prison_Break_TeddyBear_Physics
    simulationSpace=BaseBoneSpace
    alphaBoolBlend {blendOption=Linear}
    baseBoneRef {boneName=root}
    componentAppliedLinearAccClamp=<6000 6000 6000>
    componentLinearAccScale=<0.35 0.35 0.35>
    componentLinearVelScale=<0.35 0.35 0.35>
   }
   n9 = Look At - Bone: dyn_strapHipBack_l_01(ComponentPose=@n8, Alpha=0.5) {
    alphaBoolBlend {blendOption=Linear}
    boneToModify {boneName=dyn_strapHipBack_l_01}
    lookAtLocation {y=-100, z=-100}
    lookAtTarget {boneReference {boneName=dyn_teddyBody_MSTR}}
    lookAt_Axis {axis=<1 0>}
   }
   Save cached pose 'teddyCache'(Pose=Component To Local(ComponentPose=@n9)) {cacheName=teddyCache}
  ]}
 EventGraph {nodes [
   n1 = Event BlueprintUpdateAnimation()
   > Set Delta Time X(Delta Time X=@n1.DeltaTimeX)
   > n2 = Sequence!(execute) -> (then_0, then_1, then_2:Exec, then_3:Exec, then_4:Exec, then_5:Exec)
   n3 = Gravity Override State Function DEPRECATED(execute=@n2.then_1, CharacterPartAnimInstance=Self-Reference(), GameTimeInSeconds=0, GravityOverrideParams=Get Gravity Override Params())
   > Set Gravity Override(Gravity Override=@n3)
   Set WPO Globals Function(
    execute=@n2.then_2
    Material=Get Mat Instance Dynamic()
    Overlay Material=Get Mat Instance Overlay() -> (Mat Instance Overlay:Material Instance Dynamic Object Reference)
    Skeletal Mesh=Get bodyMesh() -> (bodyMesh:Skeletal Mesh Component Object Reference)
    "Local Noise Transform  Location"="Get Local Noise Transform  Location"() -> ("Local Noise Transform  Location":Vector)
    Local Noise Offset=Get Local Noise Offset()
   )
   > Set WPO Regional Function!(
    execute
    Material:Material Instance Dynamic Object Reference=Get Mat Instance Dynamic()
    Overlay Material:Material Instance Dynamic Object Reference=Get Mat Instance Overlay()
    Skeletal Mesh:Skeletal Mesh Component Object Reference=Get bodyMesh()
    Velocity Transform:Name=pelvis
    Velocity Interp Speed:double=Get Velocity Interp Speed()
    Collision Transform:Name=dyn_clothBack_02
    Collision Transform Axis:EAxis Enum
    Invert Collision Transform Axis:bool=false
    Delta Time X:double=Get Delta Time X()
    Previous Velocity Transform Location:"Vector (by ref)"=Get Previous Velocity Transform Location() -> (Previous Velocity Transform Location:Vector)
    Previous Velocity Transform Velocity:"Vector (by ref)"=Get Previous Velocity Transform Velocity() -> (Previous Velocity Transform Velocity:Vector)
    Region:int=1
   ) -> (then)
   > Set WPO Regional Function(
    Material=Get Mat Instance Dynamic()
    Overlay Material=Get Mat Instance Overlay()
    Skeletal Mesh=Get bodyMesh()
    Velocity Transform=dyn_teddyAnchor_MSTR
    Velocity Interp Speed=Get Velocity Interp Speed()
    Collision Transform=dyn_teddyBody_02
    Invert Collision Transform Axis=false
    Delta Time X=Get Delta Time X()
    Previous Velocity Transform Location=Get Previous Velocity Transform Location_0() -> (Previous Velocity Transform Location_0:Vector)
    Previous Velocity Transform Velocity=Get Previous Velocity Transform Velocity_0() -> (Previous Velocity Transform Velocity_0:Vector)
    Region=2
   )
   Set bodyMesh(execute=@n2.then_0, bodyMesh:Skeletal Mesh Component Object Reference=GetPartSkeletalMeshComponent()) -> (Output_Get:Skeletal Mesh Component Object Reference)
   n4 = PlayerMovement(execute=@n2.then_3, AnimInstance=Self-Reference(), XDotProductCoefficient=1, YDotProductCoefficient=1, ZDotProductCoefficient=1)
   n5 = Branch(execute=@n2.then_4, Condition="Not Equal (Float)"(A=@n4.YMovement))
   > Set charMoving(charMoving:bool=false) -> (Output_Get:bool)
   Set charMoving(execute=@n5.else, charMoving=true)
   Set Is Wearing Backbling(execute=@n2.then_5, Is Wearing Backbling=CheckForBackbling())
   Event OnCharacterCustomizationLoaded()
   n6 = Get bodyMesh()
   > n7 = Get Dynamic Material Instance(SkeletalMeshComponent=@n6)
   > Set Mat Instance Dynamic(Mat Instance Dynamic=@n7)
   > n8 = GetOrCreateDynamicOverlayMaterial(MeshComponent=@n6, MaterialName=None)
   > Set Mat Instance Overlay(Mat Instance Overlay:Material Instance Dynamic Object Reference=@n8) -> (Output_Get:Material Instance Dynamic Object Reference)
   Set buckle Rot Z(buckle Rot Z:float=GetAlphaFromReferenceTransformDelta(
     SkeletalMeshComponent=Get bodyMesh()
     BoneName=dyn_strapHipBack_l_01
     BaseName=dyn_strapHipBack_l_MSTR
     InputRange=(X=0,Y=1)
     TwistAxis="1,0,0"
    )) -> (Output_Get:float)
   Set isCrouchStopped(execute=@n12.then, isCrouchStopped:bool=true) -> (Output_Get:bool)
   Set isCrouchStopped(execute=@n11.then, isCrouchStopped=false)
   n9 = MapRangeClamped(Value=@n4.YMovement, InRangeA=0, InRangeB=1)
   n10 = Branch(execute=@n4.then, Condition="float > float"(A=@n9, B=0.5))
   > n11 = Set PlayerMovement(PlayerMovement:bool=true) -> (Output_Get:bool)
   n12 = Set PlayerMovement(execute=@n10.else, PlayerMovement=false)
  ]}
]
variables [
 bodyMesh {type=SkeletalMeshComponent}
 buckle Rot Z {type=number}
 charMoving {type=boolean}
 Delta Time X {type=number}
 Gravity Override {type=Vector}
 Gravity Override Params {type=GravityOverrideParamsStruct}
 Is Wearing Backbling {type=boolean}
 isCrouchStopped {type=boolean}
 Local Noise Offset {type=Vector}
 "Local Noise Transform  Location" {type=Vector}
 Mat Instance Dynamic {type=MaterialInstanceDynamic}
 Mat Instance Overlay {type=MaterialInstanceDynamic}
 PlayerMovement {type=boolean}
 Previous Velocity Transform Location {type=Vector}
 Previous Velocity Transform Location_0 {type=Vector}
 Previous Velocity Transform Velocity {type=Vector}
 Previous Velocity Transform Velocity_0 {type=Vector}
 Velocity Interp Speed {type=number, default=50}
]
