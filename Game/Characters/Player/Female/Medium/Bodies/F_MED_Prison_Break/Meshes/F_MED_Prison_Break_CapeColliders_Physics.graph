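# Physics asset for the female medium Prison Break player body, named as the set of colliders for its cape.
# Single capsules approximate the pelvis, spine, arms, hands and legs, plus dynamic belt, back-cloth and teddy bones.
# Constraints chain the bodies out from the pelvis, with the belt hanging off the back-cloth body.
# Angular motion is explicitly free only on the lower two belt joints and the spine_01-to-pelvis joint.
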PhysicsAsset /Game/Characters/Player/Female/Medium/Bodies/F_MED_Prison_Break/Meshes/F_MED_Prison_Break_CapeColliders_Physics
bodies {
 calf_l [{shapeType=Capsule, length=28.5304, radius=12.8195, center=<-27.6485 0.8852 -2.1159>, rotation=<rot 0 -100.769 -88.8272>}]
 calf_r [{shapeType=Capsule, length=28.5271, radius=12.8222, center=<27.6027 -0.9687 2.3384>, rotation=<rot 0 -100.5728 -88.8727>}]
 dyn_belt2_02 [{shapeType=Capsule, length=19.1032, radius=3.111, center=<27.9483 -0.1803 0.0624>, rotation=<rot 0 -90.9767 -89.8469>}]
 dyn_belt2_04 [{
   shapeType=Capsule
   length=3.1833
   radius=4.1749
   shapeName=dyn_belt2_04_capsule
   center=<4.8368 0.0815 -0.0006>
   rotation=<rot 0 -91.4258 -89.2498>
  }]
 dyn_belt2_06 [{
   shapeType=Capsule
   length=0.633
   radius=3.6179
   shapeName=dyn_belt2_06_capsule
   center=<3.3868 0.0164 0.0014>
   rotation=<rot -0.3964 -90 -0.8875>
  }]
 dyn_clothBack_03 [{shapeType=Capsule, length=11.2477, radius=9.1394, center=<7.9073 0.7054 -0.4708>}]
 dyn_teddyBody_01 [{shapeType=Capsule, length=11.2111, radius=9.1078, center=<-2.9442 -1.7877 2.2984>, rotation=<rot 0 -103.0229 -85.9978>}]
 dyn_teddyBody_03 [{shapeType=Capsule, length=10.6573, radius=6.5997, center=<7.2772 -0.7665 1.2879>, rotation=<rot 0 81.3851 -83.0282>}]
 hand_l [{shapeType=Capsule, length=9.201, radius=6.0137, center=<7.7863 -2.2913 0.1006>, rotation=<rot 0 81.9365 -77.1972>}]
 hand_r [{shapeType=Capsule, length=9.201, radius=6.0137, center=<-7.7863 2.2913 -0.1006>, rotation=<rot 0 81.9365 -77.1973>}]
 lowerarm_l [{shapeType=Capsule, length=14.6335, radius=7.1119, center=<8.9075 0.5941 0.8222>, rotation=<rot 0 -88.1328 -88.8572>}]
 lowerarm_r [{shapeType=Capsule, length=13.9263, radius=6.8113, center=<-9.5776 -0.3189 -0.9543>, rotation=<rot 0 -88.1374 -87.9225>}]
 pelvis [{shapeType=Capsule, length=18.374, radius=12.6583, center=<-6.4863 -3.9547 -1.3224>}]
 spine_01 [{shapeType=Capsule, length=4.255, radius=12.005, shapeName=spine_01_capsule, rotation=<rot 0 -90 0>}]
 spine_02 [{shapeType=Capsule, length=14.8115, radius=12.4148, center=<10.5906 2.9139 -0.194>}]
 spine_05 [{shapeType=Capsule, length=23.1125, radius=12.5194, center=<-4.1144 0.2848 0.0227>}]
 thigh_l [{shapeType=Capsule, length=33.1196, radius=9.8501, center=<-23.5279 0 0>, rotation=<rot 90 0 0>}]
 thigh_r [{shapeType=Capsule, length=33.1196, radius=9.8501, center=<23.5279 0 0>, rotation=<rot 90 0 180>}]
 upperarm_l [{shapeType=Capsule, length=13.3273, radius=8.6845, center=<12.0487 0.9042 -0.4276>, rotation=<rot 0 -96.3651 -87.8542>}]
 upperarm_r [{shapeType=Capsule, length=13.2774, radius=8.148, center=<-12.0553 -0.8113 0.5349>, rotation=<rot 0 -94.3419 -86.912>}]
}
constraints [
 {bone1Name=spine_02, bone2Name=spine_01, swing1LimitDegrees=45, swing2LimitDegrees=45, twistLimitDegrees=45}
 {bone1Name=spine_05, bone2Name=spine_02, swing1LimitDegrees=45, swing2LimitDegrees=45, twistLimitDegrees=45}
 {bone1Name=upperarm_l, bone2Name=spine_05, swing1LimitDegrees=45, swing2LimitDegrees=45, twistLimitDegrees=45}
 {bone1Name=lowerarm_l, bone2Name=upperarm_l, swing1LimitDegrees=45, swing2LimitDegrees=45, twistLimitDegrees=45}
 {bone1Name=hand_l, bone2Name=lowerarm_l, swing1LimitDegrees=45, swing2LimitDegrees=45, twistLimitDegrees=45}
 {bone1Name=upperarm_r, bone2Name=spine_05, swing1LimitDegrees=45, swing2LimitDegrees=45, twistLimitDegrees=45}
 {bone1Name=lowerarm_r, bone2Name=upperarm_r, swing1LimitDegrees=45, swing2LimitDegrees=45, twistLimitDegrees=45}
 {bone1Name=hand_r, bone2Name=lowerarm_r, swing1LimitDegrees=45, swing2LimitDegrees=45, twistLimitDegrees=45}
 {bone1Name=thigh_l, bone2Name=pelvis, swing1LimitDegrees=45, swing2LimitDegrees=45, twistLimitDegrees=45}
 {bone1Name=calf_l, bone2Name=thigh_l, swing1LimitDegrees=45, swing2LimitDegrees=45, twistLimitDegrees=45}
 {bone1Name=calf_r, bone2Name=thigh_r, swing1LimitDegrees=45, swing2LimitDegrees=45, twistLimitDegrees=45}
 {bone1Name=dyn_clothBack_03, bone2Name=pelvis, swing1LimitDegrees=45, swing2LimitDegrees=45, twistLimitDegrees=45}
 {bone1Name=dyn_belt2_02, bone2Name=dyn_clothBack_03, swing1LimitDegrees=45, swing2LimitDegrees=45, twistLimitDegrees=45}
 {bone1Name=dyn_teddyBody_01, bone2Name=pelvis, swing1LimitDegrees=45, swing2LimitDegrees=45, twistLimitDegrees=45}
 {bone1Name=dyn_teddyBody_03, bone2Name=dyn_teddyBody_01, swing1LimitDegrees=45, swing2LimitDegrees=45, twistLimitDegrees=45}
 {
  bone1Name=dyn_belt2_04
  bone2Name=dyn_belt2_02
  swing1LimitDegrees=45
  swing1Motion=Free
  swing2LimitDegrees=45
  swing2Motion=Free
  twistLimitDegrees=45
  twistMotion=Free
 }
 {
  bone1Name=dyn_belt2_06
  bone2Name=dyn_belt2_04
  swing1LimitDegrees=45
  swing1Motion=Free
  swing2LimitDegrees=45
  swing2Motion=Free
  twistLimitDegrees=45
  twistMotion=Free
 }
 {
  bone1Name=spine_01
  bone2Name=pelvis
  swing1LimitDegrees=45
  swing1Motion=Free
  swing2LimitDegrees=45
  swing2Motion=Free
  twistLimitDegrees=45
  twistMotion=Free
 }
 {bone1Name=thigh_r, bone2Name=pelvis, swing1LimitDegrees=45, swing2LimitDegrees=45, twistLimitDegrees=45}
]
